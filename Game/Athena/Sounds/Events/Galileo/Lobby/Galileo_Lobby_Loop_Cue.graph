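# Galileo lobby event music or ambience cue that plays a start sound and brings in looping audio behind it, on the scripted events sound class.
# The start wave plays immediately, mixed with a branch delayed by 1.5 seconds.
# In that branch, volume envelopes fade in the main loop, then crossfade it to a subtle loop between 15 and 30 seconds.
# A Looping node with an empty input also feeds the delayed mixer.
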SoundCue /Game/Athena/Sounds/Events/Galileo/Lobby/Galileo_Lobby_Loop_Cue
root=Mixer_0
nodes [
 {id=Delay_1, children [Mixer_1], properties {delayMax=1.5, delayMin=1.5}}
 {
  id=Enveloper_0
  children [WavePlayer_0]
  properties {pitchCurve {editorCurveData {keys [{time=0}]}}, volumeCurve {editorCurveData {keys [<at 0 0>,<at 1 1>,<at 15 1>,<at 30 0>]}}}
 }
 {
  id=Enveloper_1
  children [WavePlayer_6]
  properties {pitchCurve {editorCurveData {keys [{time=0}]}}, volumeCurve {editorCurveData {keys [<at 0 0>,<at 1 0>,<at 15 0>,<at 30 1>]}}}
 }
 {id=Looping_0, children [null]}
 {id=Mixer_0, children [Delay_1,WavePlayer_1], properties {inputVolume [1,1]}}
 {id=Mixer_1, children [Looping_0,Mixer_2], properties {inputVolume [1.5,1]}}
 {id=Mixer_2, children [Enveloper_0,Enveloper_1], properties {inputVolume [1,1.5]}}
 {id=WavePlayer_0, properties {bLooping=true, soundWaveAssetPtr=&/Galileo_Lobby_Loop_01}}
 {id=WavePlayer_1, properties {soundWaveAssetPtr=&/Galileo_Lobby_Start_01}}
 {id=WavePlayer_6, properties {bLooping=true, soundWaveAssetPtr=&/Galileo_Lobby_Loop_Subtle_01}}
]
properties {maxDistance=2097152, soundClassObject=/Game/Sounds/FortSoundClassesAndModes/Fort_Events_Scripted, volumeMultiplier=0.2632}
