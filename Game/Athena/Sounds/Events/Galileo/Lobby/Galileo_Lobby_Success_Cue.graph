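# Galileo_Lobby_Success_Cue is a Sound Cue for a success sound in the Galileo event lobby.
# A single wave player node plays the Galileo_Lobby_Success_01 wave directly as the root, with no randomization or modulation.
# It routes through the application tray notification sound class with a boosted volume multiplier.
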SoundCue /Game/Athena/Sounds/Events/Galileo/Lobby/Galileo_Lobby_Success_Cue
root=WavePlayer_2
nodes [{id=WavePlayer_2, properties {soundWaveAssetPtr=&/Galileo_Lobby_Success_01}}]
properties {maxDistance=2097152, soundClassObject=/Game/Sounds/FortSoundClassesAndModes/Fort_Application_TrayNotification, volumeMultiplier=1.4964}
